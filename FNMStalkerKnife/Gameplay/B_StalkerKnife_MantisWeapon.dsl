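; Stalker Knife weapon that disguises its wielder while equipped and adds a custom Tactical Sprint class.
; Equipping attaches a left-hand phone mesh, then a short timer applies the ShadowBomb effect and grants the disguise item.
; Unequipping removes both, but switching to building tools defers this via an equip listener, preserving appearance in build mode.
; It also drives flare and knife-trail Niagara effects and fires a GameplayCue when the wielder dies.
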(event Custom|StalkerDied
  (params (DamagedActor :type Actor%20Object%20Reference) (Damage :type Float%20%28single-precision%29) (InstigatedBy :type Controller%20Object%20Reference) (DamageCauser :type Actor%20Object%20Reference) (HitLocation :type Vector) (FHitComponent :type Primitive%20Component%20Object%20Reference) (BoneName :type Name) (Momentum :type Vector))
  ;[1 1 1] Player Died while holding weapon
  (Development|PrintString "Stalker Died")
  (GameplayCue|ExecuteGameplayCue%28Burst%29Local 0 (Variables|Default|GetStalkerDiedGCTag) (Ability|GameplayCue|MakeGameplayCueParameters :NormalizedMagnitude 0.0 :RawMagnitude 0.0 :Location "0,0,0" :Normal "0,0,0" :Instigator InstigatedBy :GameplayEffectLevel 1 :AbilityLevel 1 :bReplicateLocationWhenUsingMinimalRepProxy false)))

(event Custom|AttachMesh
  ;[1 1 1] Comment
  (bind _pawn (Variables|Default|GetPawn)
    (:then
      (Transformation|AttachComponentToComponent (Variables|Default|GetLeftHandPhoneMesh) (Variables|/Script/Engine.Character|GetMesh _pawn) "hand_attach_l" "SnapToTarget" "KeepRelative" "KeepRelative" true)
      (Development|SetHiddeninGame (Variables|Default|GetLeftHandPhoneMesh) false false)
      (Utilities|Time|SetTimerbyEvent (delegate Custom|PreAttachPrep) 0.2))
    (:else)))

(event EventBeginPlay
  ;[1 1 1] Begin Play
  (|CallParentFunction)
  (Development|PrintString "WEapon Class Begin Pkay")
  (bind _asfort_player_pawn (Utilities|Casting|CastToFortPlayerPawn (Game|GetInstigator)))
  (bind _output_get (Variables|Default|SetPawn _asfort_player_pawn))
  (Variables|Default|SetbListeningForEndBuild false)
  (FortPawn|BindEventtoOnWeaponEquippedDelegate _output_get (EventDispatchers|CreateEvent :!event Custom|OnAnyWeaponEquipped))
  (EventDispatchers|BindEventtoOnDied _output_get (EventDispatchers|CreateEvent :!event Custom|StalkerDied)))

(event Custom|PreAttachPrep
  ;[1 1 1] Comment
  (CallFunction|DisguiseOn))

(event FortWeapon|EventOnEquippedAndAttached
  (bind _flarefx_niagara (Variables|Default|GetFlareFX%20Niagara))
  ;[1 1 1] Wepaon Attached
  (|CallParentFunction)
  (Development|PrintString "WEapon Class EquippedandAttached")
  (Components|Activation|Activate (Variables|Default|GetFlareFX%20Niagara) false)
  (CallFunction|AttachMesh)
  (bind _output_get (Variables|Default|SetTactical%20Sprint%20Component (Actor|GetComponentbyClass (Game|GetInstigatorController) "/Script/TacticalSprintRuntime.FortPlayerControllerComponent_TacticalSprint")))
  (Sprint|AddCustomSprintMMEClass _output_get (Variables|Default|GetSprint%20MMEClass%20to%20Add))
  (bind _pawn (Variables|Default|GetPawn)
    (:then
      (if (GameplayTags|HasMatchingGameplayTag _pawn "(TagName=\"Granted.Cooldown.StalkerKnife.StalkerMode\")")
        (Components|Activation|Deactivate _flarefx_niagara)
        (else
          (Ability|Tasks|WaitGameplayTagRemoveFromActor _pawn "(TagName=\"Granted.Cooldown.StalkerKnife.StalkerMode\")"
            (:Removed
              (Components|Activation|Activate _flarefx_niagara false))))))
    (:else)))

(event FortWeapon|EventOnUnEquippedAndDetached
  ;[1 1 1] Comment
  (|CallParentFunction)
  (Utilities|FlowControl|Sequence
    (:then_0
      (if (Utilities|IsA (Weapon|GetCurrentWeapon (Variables|Default|GetPawn)) "/Script/FortniteGame.FortWeap_BuildingTool")
        (Variables|Default|SetbListeningForEndBuild true)
        (else
          (CallFunction|DisguiseOff))))
    (:then_1
      (Development|SetHiddeninGame (Variables|Default|GetLeftHandPhoneMesh) true false)
      (Components|Activation|Deactivate (Variables|Default|GetFlareFX%20Niagara)))
    (:then_2
      (GameplayEffects|RemoveActiveEffectswithTags (Ability|GetAbilitySystemComponent (Variables|Default|GetPawn)) (Variables|Default|GetSecondaryAbilityTag)))))

(event FortWeapon|EventOnLocalUnEquip
  ;[1 1 1] Comment
  (|CallParentFunction)
  (Sprint|RemoveCustomSprintMMEClass (Variables|Default|GetTactical%20Sprint%20Component) (Variables|Default|GetSprint%20MMEClass%20to%20Add)))

(event Custom|DisguiseOn
  ;[1 1 1] Disguise Toggle
  (bind _pawn (Variables|Default|GetPawn)
    (:then
      (if (GameplayTags|HasMatchingGameplayTag _pawn "(TagName=\"Granted.StalkerKnife.Equipped\")")
        (bind _disguiseagid (Variables|Default|GetDisguiseAGID)
          (:then
            (Utilities|IsValid (Inventory|FindExistingItemForDefinition :self (Fortnite|GetFortPlayerControllerfromActor _pawn) :ItemDefinition _disguiseagid :bInStorageVault false)
              (:Is%20Not%20Valid
                (GameplayEffects|ApplyGameplayEffectToSelf :self (Ability|GetAbilitySystemComponent (Variables|Default|GetPawn)) :GameplayEffectClass "/FNMStalkerKnife/Gameplay/PhoneBooth/GE_ShadowBomb.GE_ShadowBomb_C" :Level 0.0)
                (Inventory|GiveItemtoInventoryOwner (Fortnite|GetFortPlayerControllerfromActor _pawn) _disguiseagid))))
          (:else))))
    (:else)))

(event Custom|DisguiseOff
  ;[1 1 1] Disguise Toggle
  (bind _pawn (Variables|Default|GetPawn)
    (:then
      (if (not (Utilities|IsA (Weapon|GetCurrentWeapon _pawn) "/Script/FortniteGame.FortWeap_BuildingTool"))
        (bind _disguiseagid (Variables|Default|GetDisguiseAGID)
          (:then
            (Utilities|IsValid (Inventory|FindExistingItemForDefinition :self (Fortnite|GetFortPlayerControllerfromActor _pawn) :ItemDefinition _disguiseagid :bInStorageVault false)
              (:Is%20Valid
                (GameplayEffects|RemoveActiveGameplayEffectbySourceEffect (Ability|GetAbilitySystemComponent (Variables|Default|GetPawn)) "/FNMStalkerKnife/Gameplay/PhoneBooth/GE_ShadowBomb.GE_ShadowBomb_C" 0 -1)
                (Inventory|RemoveItemFromPlayer :PlayerController (Fortnite|GetFortPlayerControllerfromActor _pawn) :ItemDefinition _disguiseagid :AmountToRemove -1 :bForceRemoval true :bSkipVerbRemoval false))))
          (:else))))
    (:else)))

(event Custom|OnAnyWeaponEquipped
  (params (NewWeapon :type Fort%20Weapon%20Object%20Reference) (PrevWeapon :type Fort%20Weapon%20Object%20Reference))
  ;[1 1 1] Check for equipping non-build weapon (for preserving appearance in build mode)
  (if (Variables|Default|GetbListeningForEndBuild)
    (Utilities|IsValid NewWeapon
      (:Is%20Valid
        ;;[1 1 1] Check to see if player switched back to this weapon
        (if (== self NewWeapon)
          (Variables|Default|SetbListeningForEndBuild false)
          (else
            ;;[1 1 1] Check to see if player no longer has a build weapon equipped
            (if (Utilities|IsA NewWeapon "/Script/FortniteGame.FortWeap_BuildingTool")
              _
              (else
                _)))))
      (:Is%20Not%20Valid
        _))
    (CallFunction|DisguiseOff)
    (Variables|Default|SetbListeningForEndBuild false)))

(event EventOnAnimTrailsNotify
  (params (bActive :type Boolean))
  (bind _trailknife (Variables|Default|GetTrailKnife))
  (if bActive
    (Networking|CanExecuteCosmeticEvents
      (:True
        (Components|Activation|Activate _trailknife true)))
    (else
      (Networking|CanExecuteCosmeticEvents
        (:True
          (Components|Activation|Deactivate _trailknife))))))

(event EventOnAnimTrailsNotify(Deactivate)
  (Networking|CanExecuteCosmeticEvents
    (:True
      (Components|Activation|Deactivate (Variables|Default|GetTrailKnife)))))

(fn ConstructionScript
  (|CallParentFunction)
  (Niagara|SetNiagaraSystemAsset (Variables|Default|GetTrailKnife) (Variables|Effects|GetAnimTrail%20Niagara%20Template) true))
