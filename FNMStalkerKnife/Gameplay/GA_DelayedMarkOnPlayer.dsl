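; GA_DelayedMarkOnPlayer is a gameplay ability in the Stalker Knife feature, granted by an effect, that signals a delayed mark back to whoever applied that effect.
; On activation it stores the instigator actor from the granting effect's context.
; When the ability ends, cancelled or not, and its avatar is a valid Fort player pawn, it sends the instigator an Event.StalkerKnife.DelayedMark gameplay event carrying that pawn.
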(event Ability|EventActivateAbility
  (Variables|Default|SetInstigator (Ability|EffectContext|GetInstigatorActor (Ability|GetGrantedbyEffectContext))))

(event Ability|EventOnEndAbility
  (params (bWasCancelled :type Boolean))
  (bind _outfortplayerpawn (CallFunction|GetAvatarasValidFortPlayerPawn)
    (:ValidFortPlayerPawn
      (Ability|SendGameplayEventtoActor (Variables|Default|GetInstigator) "(TagName=\"Event.StalkerKnife.DelayedMark\")" (Utilities|Struct|MakeGameplayEventData "(TagName=\"\")" 0 _outfortplayerpawn)))))
