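; Gameplay ability for the Stalker Knife that governs how long an energy regen window stays active; no regen logic appears in this graph.
; Activation calls the parent, then listens for the Gameplay.Event.StalkerKnife.EndAbility event and starts a timer whose length comes from the EnergyRegen_Duration scalable float.
; Either the event or the timer expiring ends the ability, and the end handler only forwards to the parent without clearing the timer.
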(event Custom|PrepareEndAbility
  (Ability|EndAbility))

(event Ability|EventActivateAbility
  (|CallParentFunction)
  (Ability|Tasks|WaitGameplayEvent "(TagName=\"Gameplay.Event.StalkerKnife.EndAbility\")"
    (:then
      (Utilities|Time|SetTimerbyEvent (delegate Custom|PrepareEndAbility) (ScalableFloat|GetValueatLevel (Variables|Default|GetRow_EnergyRegen_Duration))))
    (:EventReceived
      (Ability|EndAbility))))

(event Ability|EventOnEndAbility
  (params (bWasCancelled :type Boolean))
  (|CallParentFunction bWasCancelled))
