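; Gameplay ability for the Stalker Knife lunge, extending a parent ability that supplies the activation and targeting behavior.
; Activation only forwards to the parent, and the end-ability event is a disabled stub.
; The GetTargetingHit override takes the parent's hit result and, for a valid Fort player pawn avatar, snaps pawn and control rotation toward the hit location.
; It then returns that hit unchanged, so targeting also turns the player as a side effect.
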(event Ability|EventActivateAbility
  (|CallParentFunction))

(event Ability|EventOnEndAbility
  (params (bWasCancelled :type Boolean))
  (|CallParentFunction bWasCancelled) ; This node is disabled and will not be called.\nDrag off pins to build functionality.
)

(fn GetTargetingHit
  (bind _targetinghitresult (|CallParentFunction))
  (bind (_bblockinghit _binitialoverlap _time _distance _location) (Collision|BreakHitResult _targetinghitresult))
  (bind _outfortplayerpawn (CallFunction|GetAvatarasValidFortPlayerPawn)
    (:ValidFortPlayerPawn
      (FortPawn|SetPawnandControlRotation _outfortplayerpawn (Math|Rotator|FindLookatRotation (Transformation|GetActorLocation _outfortplayerpawn) _location))
      (return _targetinghitresult))))
