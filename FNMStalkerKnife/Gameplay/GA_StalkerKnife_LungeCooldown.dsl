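; Gameplay ability that runs the lunge cooldown for the Stalker Knife.
; On activation it removes the Granted.Cooldown.StalkerKnife.Lunge loose tag from the avatar pawn, then waits for a duration read from a scalable float.
; PrepareEndAbility then re-adds the tag, so the tag is absent during cooldown.
; If adding fails, a short looping timer retries until it succeeds, then the ability ends and the stored timer is cleared.
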(event Custom|PrepareEndAbility
  (bind _returnvalue (Ability|GameplayEffect|AddLooseGameplayTags (CallFunction|GetAvatarasFortPlayerPawn) (GameplayTags|MakeLiteralGameplayTagContainer "(GameplayTags=((TagName=\"Granted.Cooldown.StalkerKnife.Lunge\")))")))
  (if _returnvalue
    (Ability|EndAbility)
    (Utilities|Time|ClearandInvalidateTimerbyHandle (Variables|Default|GetLoopTH))
    (else
      (bind _returnvalue_2 (Utilities|Time|SetTimerbyEvent (delegate Custom|PrepareEndAbility) 0.1 true))
      (Variables|Default|SetLoopTH _returnvalue_2))))

(event Ability|EventActivateAbility
  (|CallParentFunction)
  (Ability|GameplayEffect|RemoveLooseGameplayTags (CallFunction|GetAvatarasFortPlayerPawn) (GameplayTags|MakeLiteralGameplayTagContainer "(GameplayTags=((TagName=\"Granted.Cooldown.StalkerKnife.Lunge\")))"))
  (Utilities|Time|SetTimerbyEvent (delegate Custom|PrepareEndAbility) (ScalableFloat|GetValueatLevel (Variables|Default|GetRow_Lunge_Cooldown))))
