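; Gameplay ability for the Stalker Knife that watches a marked player for death or being downed (DBNO).
; On activation it binds the pawn's OnDied and waits for the DBNO event; after two seconds it watches for marked-tag removal to end itself.
; Death fires a dead-tone gameplay cue; DBNO does so only if the instigator's inventory has the equipped-knife tagged item.
; Ending unbinds OnDied, and both handlers print debug strings.
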(event Ability|EventActivateAbility
  ; Comment
  (bind _outfortplayerpawn (CallFunction|GetAvatarasValidFortPlayerPawn)
    (:ValidFortPlayerPawn
      (bind _output_get (Variables|Default|Set%20Fort%20Player%20Pawn _outfortplayerpawn))
      (EventDispatchers|BindEventtoOnDied _output_get (EventDispatchers|CreateEvent :!event Custom|MarkedPlayerDied))
      (bind (_asynctaskproxy _payload) (Ability|Tasks|WaitGameplayEvent "(TagName=\"Event.DBNOAthena\")")
        (:then
          (Utilities|Time|SetTimerbyEvent (delegate Custom|Delay%20CheckTag) 2.0))
        (:EventReceived
          (CallFunction|MarkedPlayerDBNO _payload))))))

(event Custom|Delay%20CheckTag
  ; Comment
  (Ability|Tasks|WaitGameplayTagRemoveFromActor (stale Ability|EventActivateAbility _output_get) (Variables|Default|GetMarkedPlayerTag)
    (:Removed
      (Ability|EndAbility))))

(event Ability|EventOnEndAbility
  (params (bWasCancelled :type Boolean))
  ; Comment
  (bind _fort_player_pawn (Variables|Default|Get%20Fort%20Player%20Pawn)
    (:then
      (EventDispatchers|UnbindEventfromOnDied _fort_player_pawn (EventDispatchers|CreateEvent :!event Custom|MarkedPlayerDied)))
    (:else)))

(event Custom|MarkedPlayerDied
  (params (DamagedActor :type Actor%20Object%20Reference) (Damage :type Float%20%28single-precision%29) (InstigatedBy :type Controller%20Object%20Reference) (DamageCauser :type Actor%20Object%20Reference) (HitLocation :type Vector) (FHitComponent :type Primitive%20Component%20Object%20Reference) (BoneName :type Name) (Momentum :type Vector))
  ; Comment
  (Development|PrintString (Utilities|String|Append "Marked Player Died " " Instigator: " (Utilities|GetDisplayName InstigatedBy)) true true "(R=1.000000,G=0.071657,B=0.015760,A=1.000000)" 1e+01)
  (bind _fort_player_pawn (Variables|Default|Get%20Fort%20Player%20Pawn)
    (:then
      (GameplayCue|ExecuteGameplayCueOnActor(Burst) _fort_player_pawn (Variables|Default|GetDeadToneCueTag) (Ability|GameplayCue|MakeGameplayCueParameters :NormalizedMagnitude 0.0 :RawMagnitude 0.0 :Location "0,0,0" :Normal "0,0,0" :Instigator InstigatedBy :GameplayEffectLevel 1 :AbilityLevel 1 :bReplicateLocationWhenUsingMinimalRepProxy false)))
    (:else)))

(event Custom|Marked%20Player%20DBNO
  (params (Payload :type Gameplay%20Event%20Data%20Structure))
  (bind (_eventtag _instigator) (Utilities|Struct|BreakGameplayEventData Payload))
  ; Comment
  (Development|PrintString (Utilities|String|Append "Marked Player DBNO " " Instigator: " (Utilities|GetDisplayName _instigator)) true true "(R=1.000000,G=0.071657,B=0.015760,A=1.000000)" 1e+01)
  (bind _returnvalue (Inventory|GetInventoryItemwithTag (Fortnite|GetFortPlayerControllerfromActor _instigator) (GameplayTags|MakeLiteralGameplayTag "(TagName=\"Granted.StalkerKnife.Equipped\")")))
  (Utilities|IsValid _returnvalue
    (:Is%20Valid
      (bind _fort_player_pawn (Variables|Default|Get%20Fort%20Player%20Pawn)
        (:then
          (GameplayCue|ExecuteGameplayCueOnActor(Burst) _fort_player_pawn (Variables|Default|GetDeadToneCueTag) (Ability|GameplayCue|MakeGameplayCueParameters :NormalizedMagnitude 0.0 :RawMagnitude 0.0 :Location "0,0,0" :Normal "0,0,0" :Instigator _instigator :GameplayEffectLevel 1 :AbilityLevel 1 :bReplicateLocationWhenUsingMinimalRepProxy false)))
        (:else)))))
