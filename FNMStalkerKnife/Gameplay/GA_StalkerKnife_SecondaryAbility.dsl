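; Stalker Knife secondary gameplay ability that triggers a scan and turns on stalk mode.
; Activation plays the detonator montage, deactivates the weapon's flare FX, applies the cooldown tag effect and commits.
; A short timer then fires the scanning cue, applies StalkModeOn and spends weapon durability.
; The ability stays active until the cooldown tag is removed, sends an event on ending, and plays a sound when activation fails on cooldown.
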(event Ability|EventActivateAbility
  ; Activate Ability
  (bind _output_get (Variables|Default|SetPawn (CallFunction|GetAvatarasFortPlayerPawn)))
  (Variables|Default|SetPlayerController (Fortnite|GetFortPlayerControllerfromActor _output_get))
  (Ability|Tasks|PlayMontageAndWait "None" "/FNMStalkerKnife/Assets/ANM/Ability_StalkerKnife_Detonator_Activate.Ability_StalkerKnife_Detonator_Activate")
  ;;[1 0 0] FX DEACTIVATION ANTENNA FLARE IDLE DURING COOLDOWN
  (bind _asb_stalker_knife_mantis_weapon (Utilities|Casting|CastToB_StalkerKnife_MantisWeapon (Weapon|GetCurrentWeapon (Variables|Default|GetPawn))))
  (bind _output_get_2 (Variables|Default|SetFlare%20FX%20Niagara (Variables|/FNMStalkerKnife/Gameplay/B_StalkerKnife_MantisWeapon.B_StalkerKnife_MantisWeapon_C|GetFlareFX%20Niagara _asb_stalker_knife_mantis_weapon)))
  (Components|Activation|Deactivate _output_get_2)
  (bind _returnvalue (Ability|ApplyGameplayEffectToOwner "/FNMStalkerKnife/Gameplay/GE_StalkerKnife_SecondaryAbility_GrantCooldownTag.GE_StalkerKnife_SecondaryAbility_GrantCooldownTag_C"))
  (Variables|Default|SetSecondaryAbilityGEHandle _returnvalue)
  (Ability|CommitAbility)
  (Utilities|Time|SetTimerbyEvent (EventDispatchers|CreateEvent :!event Custom|DelayedMark) 0.8)
  (bind _pawn (Variables|Default|GetPawn)
    (:then
      (Ability|Tasks|WaitGameplayTagRemoveFromActor _pawn "(TagName=\"Granted.Cooldown.StalkerKnife.StalkerMode\")"
        (:Removed
          (Ability|EndAbility))))
    (:else)))

(event Custom|DelayedMark
  ; Delayed Mark and PreEndAbility Logic
  (bind _outfortplayerpawn (CallFunction|GetAvatarasValidFortPlayerPawn)
    (:ValidFortPlayerPawn
      (Ability|ExecuteGameplayCueWithParamsOnOwner (Variables|Default|GetScanning%20Gameplay%20Cue) (Ability|GameplayCue|MakeGameplayCueParameters :NormalizedMagnitude 0.0 :RawMagnitude 0.0 :Location (Transformation|GetActorLocation _outfortplayerpawn) :Normal "0,0,0" :GameplayEffectLevel 1 :AbilityLevel 1 :bReplicateLocationWhenUsingMinimalRepProxy false))
      (Ability|ApplyGameplayEffectToOwner "/FNMStalkerKnife/Gameplay/GE_StalkerKnife_StalkModeOn.GE_StalkerKnife_StalkModeOn_C")
      (bind _pawn (Variables|Default|GetPawn)
        (:then
          (bind _asb_stalker_knife_mantis_weapon (Utilities|Casting|CastToB_StalkerKnife_MantisWeapon (Weapon|GetCurrentWeapon _pawn)))
          (FortWeapon|UseWeaponDurability(Explicit) _asb_stalker_knife_mantis_weapon (ScalableFloat|GetValueatLevel (Variables|Default|GetHotfix_HitDurabilityCost))))
        (:else)))))

(event Ability|EventOnEndAbility
  (params (bWasCancelled :type Boolean))
  ; ON END ABILITY
  (Ability|SendGameplayEventtoActor (CallFunction|GetAvatarasFortPlayerPawn) "(TagName=\"Granted.StalkerOTL.Ability.End\")"))

(fn OverrideFailedReason
  (params (FailedReason :type Gameplay%20Tag%20Container%20Structure%20%28by%20ref%29))
  (Utilities|FlowControl|Sequence
    (:then_0
      (if (GameplayTags|HasTag FailedReason "(TagName=\"Activate.Failed.Cooldown\")")
        (Audio|PlaySound2D (Variables|Default|GetAbilityOnCooldownSound))
        (Development|PrintString "Can't USE SOUND SHOULD BE PLAYING!")))
    (:then_1
      (bind _overridenfailedreason (|CallParentFunction FailedReason))
      (return _overridenfailedreason))))
