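; Guided-motion movement mode extension for the Stalker Knife dash, driving the owner pawn with a constant force.
; Direction is the forward vector of the pawn's "PawnForward" targeting transform, computed identically in both branches of a Walking tag check.
; Duration is hotfixable targeting length divided by force strength, and finish velocity is fixed along the dash direction.
; Activation caches a gravity multiplier; height enforcement follows local height, with target height zero.
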(event EventOnExtensionActivate
  (params (RuntimeContext :type MMERuntime%20Context%20Structure%20%28by%20ref%29))
  (Variables|Default|SetCachedGravityScaleModifier 1.25))

(fn GetDirection_ConstantForce
  (params (GuidedMotionRuntimeData :type Fort%20Movement%20Mode%20Guided%20Motion%20Runtime%20Data%20Object%20Reference))
  (bind (_returnvalue_location _returnvalue_rotation) (Math|Transform|BreakTransform (Ability|GetTargetingTransform (Variables|FortMovementMode_BaseExtLogic|GetOwnerPawn) "PawnForward" 0)))
  (bind (_returnvalue_location_2 _returnvalue_rotation_2) (Math|Transform|BreakTransform (Ability|GetTargetingTransform (Variables|FortMovementMode_BaseExtLogic|GetOwnerPawn) "PawnForward" 0)))
  (if (GameplayTags|HasMatchingGameplayTag (Variables|FortMovementMode_BaseExtLogic|GetOwnerPawn) "(TagName=\"MovementMode.Walking\")")
    (return (Math|Vector|GetForwardVector _returnvalue_rotation))
    (else
      (return (Math|Vector|GetForwardVector _returnvalue_rotation_2)))))

(fn GetEnforceHeightAboveGround_ConstantForce
  (params (GuidedMotionRuntimeData :type Fort%20Movement%20Mode%20Guided%20Motion%20Runtime%20Data%20Object%20Reference))
  (return (> (Variables|/Script/FortniteGame.FortMovementMode_GuidedMotionRuntimeData|GetLocalHeightAboveGround GuidedMotionRuntimeData) 0)))

(fn GetFinishVelocityToSet
  (params (GuidedMotionRuntimeData :type Fort%20Movement%20Mode%20Guided%20Motion%20Runtime%20Data%20Object%20Reference))
  (return (* (Math|Vector|Normalize (Variables|/Script/FortniteGame.FortMovementMode_GuidedMotionRuntimeData|GetLocalDirection GuidedMotionRuntimeData)) 1.2e+03)))

(fn GetHeightAboveGround_ConstantForce
  (params (GuidedMotionRuntimeData :type Fort%20Movement%20Mode%20Guided%20Motion%20Runtime%20Data%20Object%20Reference))
  (return 0.0))

(fn GetMaxDuration
  (params (GuidedMotionRuntimeData :type Fort%20Movement%20Mode%20Guided%20Motion%20Runtime%20Data%20Object%20Reference))
  (return (Math|Float|SafeDivide (ScalableFloat|GetValueasDouble (Variables|Default|GetHotfix_TargetingLength)) (ScalableFloat|GetValueatLevel (Variables|Config-ConstantForce|GetStrength)))))

(fn GetModifiedGravityZScalar
  (params (GuidedMotionRuntimeData :type Fort%20Movement%20Mode%20Guided%20Motion%20Runtime%20Data%20Object%20Reference) (CurrentGravityZValueScalar :type Float%20%28single-precision%29))
  (return (* CurrentGravityZValueScalar (Variables|Default|GetCachedGravityScaleModifier))))
